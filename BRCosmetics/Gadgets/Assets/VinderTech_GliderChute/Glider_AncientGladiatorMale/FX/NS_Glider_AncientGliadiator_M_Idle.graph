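# Niagara system for the Ancient Gladiator male glider's idle effect, with purple-blue fire rings, smoke, lights, paw sprites and ribbon trails.
# One-shot deploy emitters play first, while looping emitters start after a delay.
# Rings, lights and smoke track bones through skeletal mesh sampling, and ribbons sample a second mesh input.
# An FN_Integrator value scales spawn rates, velocities and fades, while front-end modifiers adjust timing and offsets for menus.
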NiagaraSystem /BRCosmetics/Gadgets/Assets/VinderTech_GliderChute/Glider_AncientGladiatorMale/FX/NS_Glider_AncientGliadiator_M_Idle
dependencies {
 usedDataInterfaces [&NiagaraDataInterfaceCurve, &NiagaraDataInterfaceVector2DCurve, &NiagaraDataInterfaceColorCurve]
 usedDynamicInputs [&MakeFloatFromVector]
 usedModules [&SystemState, &FN_Integrator]
}
emitters {
 Decal001 {
  emitterUpdateScript [
   &EmitterState(
    Delay First Loop Only=true
    Life Cycle Mode=::NewEnumerator1
    Loop Delay=&FN_FrontEndFloatModifier(Gameplay=1.3542, Front End=0.8)
    Loop Duration=1
    UseLoopDelay=true
   )
   &SpawnRate(SpawnRate=13, Use Spawn Probability=false)
  ]
  particleSpawnScript [
   /Niagara/Modules/Spawn/Initialization/InitializeParticle(
    Color=<rgba 1.4438 0 3 5>
    Lifetime=1.25
    Module.Write Color=true
    Module.Write Lifetime=true
    Module.Write Mass=true
    Module.Write Material Random=true
    Module.Write Position=false
    Module.Write Scale=false
    Module.Write SpriteRotation=false
    Module.Write SpriteSize=true
    Position=&SimulationPosition()
    Sprite Size=&Vector2DFromFloat(Value=&RandomRangeFloat(Minimum=40, Maximum=60))
   )
   &BoxLocation(
    Box Offset=&Add_Vector(A=<0 50 60>, B=&FN_FrontEndVector3Modifier(Gameplay=<0 0 0>, Front End=<0 -60 -80>))
    Box Size=<50 0 80>
    MaskSpawnGroup=false
    Module.Override Seed=false
    Rotate Box=false
    Surface Only=false
   )
   &AddVelocity(Velocity=&RandomRangeVector(Minimum=<0 75 0>, Maximum=<1 150 1>))
   &DynamicMaterialParameters(
    Index 0 Param 1=1
    Index 0 Param 2=1.5
    Index 0 Param 3=&RandomRangeFloat(Minimum=0.1, Maximum=1)
    Index 0 Param 4=&RandomRangeFloat(Minimum=0.75, Maximum=1)
    Index 1 Param 1=0.5
    Index 1 Param 2=24
    Module.Param0WriteEnabled=true
    Module.Param0WriteEnabled1=true
    Module.Param1WriteEnabled=true
    Module.Param1WriteEnabled1=true
    Module.Param2WriteEnabled=true
    Module.Param2WriteEnabled1=true
    Module.Param3WriteEnabled=true
    Module.Param3WriteEnabled1=true
    Write Parameter Index 1=true
   )
  ]
  particleUpdateScript [
   &SolveForcesAndVelocity(Clamp Velocity=false, Limit Acceleration=false)
   &ParticleState()
   &ScaleColor(
    Scale Alpha=&FloatFromCurve(FloatCurve=DataInterface() {propertyValues {Curve {keys [<at 0 0 RCIM_Cubic>,<at 0.2 1 RCIM_Cubic>,<at 1 0 RCIM_Cubic>]}}}, Scale Curve=0.5)
    Scale RGB=&VectorFromFloat(Value=&FloatFromCurve(FloatCurve=DataInterface() {propertyValues {Curve {keys [
          <at 0 0 RCIM_Cubic>
          {arriveTangent=9.2134, interpMode=RCIM_Cubic, leaveTangent=9.2134, tangentMode=RCTM_User, time=0.1, value=3}
          <at 0.7 2>
         ]}}}))
    ScaleA=true
    ScaleRGB=true
   )
   &CameraOffset(Camera Offset Amount=&RandomRangeFloat(Minimum=8, Maximum=15))
   &ScaleSpriteSize(Scale Factor=&Vector2DFromFloat(Value=&FloatFromCurve(FloatCurve=DataInterface() {propertyValues {Curve {keys [<at 0 0>,<at 0.1 1>,<at 1 1.5>]}}})))
  ]
  rendererClasses [&NiagaraSpriteRendererProperties]
  renderers [{rendererClass=&NiagaraSpriteRendererProperties}]
 }
 FireRing {
  emitterUpdateScript [
   &EmitterState(
    Delay First Loop Only=true
    Inactive Response=::NewEnumerator1
    Life Cycle Mode=::NewEnumerator1
    Loop Delay=&FN_FrontEndFloatModifier(Gameplay=1.3542, Front End=0.8)
    Loop Duration=1
    UseLoopDelay=true
   )
   &SpawnBurst_Instantaneous(Spawn Count=2, Use Spawn Probability=false)
  ]
  particleSpawnScript [
   &FN_SetSpawnGroup(Number of Spawn Groups=2)
   /Niagara/Modules/Spawn/Initialization/V2/InitializeParticle(
    Color Mode=::NewEnumerator0
    Mesh Scale=&MaskVectorBySpawnGroup(Masked Vector=<-1 1 1>, Passthrough Vector=<1 1 1>)
    Mesh Scale Mode=::NewEnumerator4
    UsePositionOffset=false
    Write Material Random=true
   )
   &SkeletalMeshLocation(
    Apply Sampled Orientation to Mesh Particles=true
    Bone / Socket Index=&SpawnGroup()
    Bone Sampling Mode=::NewEnumerator4
    Inherit Velocity=false
    MaskSpawnGroup=false
    Skeletal Mesh=$User.SkeletalMeshONE
    Use Sampled Mesh Attribute as Ribbon ID=false
   ) {module=SkeletalMeshLocation001}
   &Color(Color=<rgba 1.2998 0 3 1>)
   &DynamicMaterialParameters(
    Index 0 Param 3=40
    Module.Param0WriteEnabled=true
    Module.Param1WriteEnabled=true
    Module.Param2WriteEnabled=true
    Module.Param3WriteEnabled=true
   )
  ]
  particleUpdateScript [
   &ParticleState()
   &SkeletalMeshLocation(
    Apply Sampled Orientation to Mesh Particles=true
    Bone / Socket Index=&SpawnGroup()
    Bone Sampling Mode=::NewEnumerator4
    Inherit Velocity=false
    MaskSpawnGroup=false
    Skeletal Mesh=$User.SkeletalMeshONE
    Use Sampled Mesh Attribute as Ribbon ID=false
   )
  ]
  rendererClasses [&NiagaraMeshRendererProperties]
  renderers [{rendererClass=&NiagaraMeshRendererProperties, rendererIndex=0}]
 }
 FireRing_Deploy {
  emitterUpdateScript [
   &EmitterState(
    Inactive Response=::NewEnumerator1
    Life Cycle Mode=::NewEnumerator1
    Loop Behavior=::NewEnumerator1
    Loop Delay=&FN_FrontEndFloatModifier(Gameplay=1.3542, Front End=0.8)
    Loop Duration=2
    UseLoopDelay=true
   )
   &SpawnBurst_Instantaneous(Spawn Count=1, Use Spawn Probability=false)
   &SpawnBurst_Instantaneous(Spawn Count=1, Spawn Group=1, Use Spawn Probability=false) {module=SpawnBurst_Instantaneous001}
  ]
  particleSpawnScript [
   &FN_SetSpawnGroup(Number of Spawn Groups=2)
   &InitializeParticle(
    Color=<rgba 3.0759 0 5 1>
    Lifetime=0.5
    Mesh Scale=&MaskVectorBySpawnGroup(Masked Vector=<-1 1 1>, Passthrough Vector=<1 1 1>)
    Mesh Scale Mode=::NewEnumerator4
    UsePositionOffset=false
    Write Material Random=true
   )
   &SkeletalMeshLocation(
    Apply Sampled Orientation to Mesh Particles=true
    Bone / Socket Index=&SpawnGroup()
    Bone Sampling Mode=::NewEnumerator4
    Inherit Velocity=false
    MaskSpawnGroup=false
    Skeletal Mesh=$User.SkeletalMeshONE
    Use Sampled Mesh Attribute as Ribbon ID=false
   ) {module=SkeletalMeshLocation001}
   &DynamicMaterialParameters(
    Index 0 Param 3=100
    Module.Param0WriteEnabled=true
    Module.Param1WriteEnabled=true
    Module.Param2WriteEnabled=true
    Module.Param3WriteEnabled=true
   )
  ]
  particleUpdateScript [
   &ParticleState()
   &SkeletalMeshLocation(
    Apply Sampled Orientation to Mesh Particles=true
    Bone / Socket Index=&SpawnGroup()
    Bone Sampling Mode=::NewEnumerator4
    Inherit Velocity=false
    MaskSpawnGroup=false
    Skeletal Mesh=$User.SkeletalMeshONE
    Use Sampled Mesh Attribute as Ribbon ID=false
   )
   &Color(Color=&ColorFromCurve(ColorCurve=DataInterface() {propertyValues {
       AlphaCurve {keys [<at 0 0 RCIM_Cubic>,<at 0.2447 1>,<at 1 0 RCIM_Cubic>]}
       BlueCurve {keys [<at 0 5 RCIM_Cubic>]}
       GreenCurve {keys [<at 0 0 RCIM_Cubic>]}
       RedCurve {keys [<at 0 3.0759 RCIM_Cubic>]}
      }})) {module=Color001}
  ]
  rendererClasses [&NiagaraMeshRendererProperties]
  renderers [{rendererClass=&NiagaraMeshRendererProperties, rendererIndex=0}]
 }
 FireSmoke {
  emitterUpdateScript [
   &EmitterState(
    Delay First Loop Only=true
    Inactive Response=::NewEnumerator1
    Life Cycle Mode=::NewEnumerator1
    Loop Delay=1
    Loop Duration=1
    UseLoopDelay=true
   )
   &SpawnRate(SpawnRate=16, Use Spawn Probability=false)
  ]
  particleSpawnScript [
   /Niagara/Modules/Spawn/Initialization/InitializeParticle(
    Color=<rgba 3.0759 0 5 0.15>
    Lifetime=&UniformRangedFloat(Minimum=1, Maximum=2)
    Module.Write Color=true
    Module.Write Lifetime=true
    Module.Write Mass=true
    Module.Write Material Random=true
    Module.Write Position=false
    Module.Write Scale=false
    Module.Write SpriteRotation=true
    Module.Write SpriteSize=true
    Position=&SimulationPosition()
    Sprite Rotation=&UniformRangedFloat(Minimum=0, Maximum=300)
    Sprite Size=&Vector2DFromFloat(Value=&UniformRangedFloat(Minimum=60, Maximum=100))
   )
   &FN_SetSpawnGroup(Number of Spawn Groups=2)
   /SetVariables(Particles.Y:NiagaraFloat=&RandomRangeFloat(Minimum=-15, Maximum=15)) {isSetParameters=true, module=SetVariables_#1}
   /SetVariables(Particles.PositionOffset:Vector3f=&MaskVectorBySpawnGroup(
     Masked Vector=&MakeVector(
      X=&RandomRangeFloat(Minimum=10, Maximum=35)
      Y=$Particles.Y
      Z=&FloatFromCurve(FloatCurve=DataInterface() {propertyValues {Curve {keys [<at -15 -40>,<at 15 30>]}}}, CurveIndex=$Particles.Y)
     )
     Passthrough Vector=&MakeVector(
      X=&RandomRangeFloat(Minimum=-10, Maximum=-35)
      Y=$Particles.Y
      Z=&FloatFromCurve(FloatCurve=DataInterface() {propertyValues {Curve {keys [<at -15 -40>,<at 15 30>]}}}, CurveIndex=$Particles.Y)
     )
    )) {isSetParameters=true, module=SetVariables_#2}
  ]
  particleUpdateScript [
   &ParticleState()
   &SkeletalMeshLocation(
    Apply Sampled Orientation to Mesh Particles=true
    Bone / Socket Index=&SpawnGroup()
    Bone Sampling Mode=::NewEnumerator4
    Inherit Velocity=false
    MaskSpawnGroup=false
    Sampled Position Offset=&Add_Vector(
     A=$Particles.PositionOffset
     B=&Lerp_Vector(
      A=<0 0 0>
      B=&RandomRangeVector(Minimum=<0 0 0>, Maximum=<0 130 0>)
      Alpha=&FloatFromCurve(FloatCurve=DataInterface() {propertyValues {Curve {keys [<at 0 0 RCIM_Cubic>, {arriveTangent=2.4246, interpMode=RCIM_Cubic, leaveTangent=2.4246, tangentMode=RCTM_User, time=1, value=1}]}}})
     )
    )
    Sampled Position Offset Coordinate Space:ENiagaraMeshTransforms=::NewEnumerator2
    Skeletal Mesh=$User.SkeletalMeshONE
   )
   &ScaleColor(
    Scale Alpha=&FloatFromCurve(
     FloatCurve=DataInterface() {propertyValues {Curve {keys [
         {arriveTangent=6.1149, interpMode=RCIM_Cubic, leaveTangent=6.1149, tangentMode=RCTM_User, time=0, value=0}
         {interpMode=RCIM_Cubic, tangentMode=RCTM_User, time=0.2, value=1}
         {arriveTangent=-1.2458, interpMode=RCIM_Cubic, leaveTangent=-1.2458, time=0.7303, value=0.3515}
         {interpMode=RCIM_Cubic, leaveTangent=-4.41, time=1.0011, value=0.002}
        ]}}}
     Scale Curve=&OneMinusFloat(Float=$System.FN_Integrator001.Value)
    )
    ScaleA=true
    ScaleRGB=false
   )
   &SolveForcesAndVelocity(Clamp Velocity=false, Limit Acceleration=false)
   &ScaleSpriteSize(Scale Factor=&Vector2DFromFloat(Value=&FloatFromCurve(FloatCurve=DataInterface() {propertyValues {Curve {keys [
          {arriveTangent=6.3444, interpMode=RCIM_Cubic, time=0, value=0.25}
          {arriveTangent=1.3547, interpMode=RCIM_Cubic, leaveTangent=1.3547, tangentMode=RCTM_User, time=0.2884, value=0.5549}
          {interpMode=RCIM_Cubic, leaveTangent=-1.768, time=0.9996, value=1.2}
         ]}}})))
   &SubUVAnimation(
    Sub UV Play Rate:NiagaraFloat=&RandomRangeFloat(Minimum=0.4, Maximum=0.7)
    SubUV Animation Mode:ENiagara_SubUVLookupMode=::NewEnumerator3
    SubUV Start Time Offset:NiagaraFloat=&RandomRangeFloat(Minimum=0, Maximum=1)
   )
   &DynamicMaterialParameters(
    Index 0 Param 1=&FloatFromCurve(FloatCurve=DataInterface() {propertyValues {Curve {keys [{interpMode=RCIM_Cubic, tangentMode=RCTM_User, time=0.4, value=0}, {interpMode=RCIM_Cubic, tangentMode=RCTM_User, time=1, value=-0.75}]}}})
    Index 0 Param 4=1
    Module.Param0WriteEnabled=true
    Module.Param1WriteEnabled=false
    Module.Param2WriteEnabled=false
    Module.Param3WriteEnabled=true
   )
  ]
  rendererClasses [&NiagaraSpriteRendererProperties]
  renderers [{rendererClass=&NiagaraSpriteRendererProperties}]
 }
 FireSmoke_Forward_Deploy {
  emitterUpdateScript [
   &EmitterState(
    Life Cycle Mode=::NewEnumerator1
    Loop Behavior=::NewEnumerator1
    Loop Delay=&FN_FrontEndFloatModifier(Gameplay=1.2, Front End=0.55)
    Loop Duration=1
    UseLoopDelay=true
   )
   &SpawnBurst_Instantaneous(Spawn Count=22, Use Spawn Probability=false)
  ]
  particleSpawnScript [
   &InitializeParticle(
    Color=<rgba 6.1519 0 10 0.15>
    Lifetime=&UniformRangedFloat(Minimum=0.75, Maximum=1.1)
    Module.Write Color=true
    Module.Write Lifetime=true
    Module.Write Mass=true
    Module.Write Material Random=true
    Module.Write Position=false
    Module.Write Scale=false
    Module.Write SpriteRotation=true
    Module.Write SpriteSize=true
    Position=&SimulationPosition()
    Sprite Rotation=&UniformRangedFloat(Minimum=0, Maximum=300)
    Sprite Size=&Vector2DFromFloat(Value=&UniformRangedFloat(Minimum=80, Maximum=100))
   )
   &FN_SetSpawnGroup(Number of Spawn Groups=2)
   /SetVariables(Particles.Y:NiagaraFloat=&RandomRangeFloat(Minimum=-15, Maximum=15)) {isSetParameters=true, module=SetVariables_#1}
   /SetVariables(Particles.PositionOffset:Vector3f=&MaskVectorBySpawnGroup(
     Masked Vector=&MakeVector(
      X=&RandomRangeFloat(Minimum=10, Maximum=35)
      Y=$Particles.Y
      Z=&FloatFromCurve(FloatCurve=DataInterface() {propertyValues {Curve {keys [<at -15 -40>,<at 15 30>]}}}, CurveIndex=$Particles.Y)
     )
     Passthrough Vector=&MakeVector(
      X=&RandomRangeFloat(Minimum=-10, Maximum=-35)
      Y=$Particles.Y
      Z=&FloatFromCurve(FloatCurve=DataInterface() {propertyValues {Curve {keys [<at -15 -40>,<at 15 30>]}}}, CurveIndex=$Particles.Y)
     )
    )) {isSetParameters=true, module=SetVariables_#2}
  ]
  particleUpdateScript [
   &ParticleState()
   &SkeletalMeshLocation(
    Apply Sampled Orientation to Mesh Particles=true
    Bone / Socket Index=&SpawnGroup()
    Bone Sampling Mode=::NewEnumerator4
    Inherit Velocity=false
    MaskSpawnGroup=false
    Sampled Position Offset=&Add_Vector(
     A=$Particles.PositionOffset
     B=&Lerp_Vector(
      A=<0 0 0>
      B=&RandomRangeVector(Minimum=<-50 0 40>, Maximum=<50 -250 100>)
      Alpha=&FloatFromCurve(FloatCurve=DataInterface() {propertyValues {Curve {keys [
           {arriveTangent=0.4807, interpMode=RCIM_Cubic, leaveTangent=0.4807, tangentMode=RCTM_User, time=0, value=0}
           {arriveTangent=1.6838, interpMode=RCIM_Cubic, leaveTangent=1.6838, tangentMode=RCTM_User, time=1, value=1}
          ]}}})
     )
    )
    Sampled Position Offset Coordinate Space:ENiagaraMeshTransforms=::NewEnumerator2
    Skeletal Mesh=$User.SkeletalMeshONE
   )
   &ScaleSpriteSize(Scale Factor=&Vector2DFromFloat(Value=&FloatFromCurve(FloatCurve=DataInterface() {propertyValues {Curve {keys [
          {arriveTangent=6.3444, interpMode=RCIM_Cubic, time=0, value=0.25}
          {arriveTangent=1.3547, interpMode=RCIM_Cubic, leaveTangent=1.3547, tangentMode=RCTM_User, time=0.2884, value=0.5549}
          {interpMode=RCIM_Cubic, leaveTangent=-1.768, time=0.9996, value=1.2}
         ]}}})))
   &SubUVAnimation(
    Sub UV Play Rate:NiagaraFloat=&RandomRangeFloat(Minimum=0.4, Maximum=0.7)
    SubUV Animation Mode:ENiagara_SubUVLookupMode=::NewEnumerator3
    SubUV Start Time Offset:NiagaraFloat=&RandomRangeFloat(Minimum=0, Maximum=1)
   )
   &DynamicMaterialParameters(
    Index 0 Param 1=&FloatFromCurve(FloatCurve=DataInterface() {propertyValues {Curve {keys [{interpMode=RCIM_Cubic, tangentMode=RCTM_User, time=0.4, value=0}, {interpMode=RCIM_Cubic, tangentMode=RCTM_User, time=1, value=-0.75}]}}})
    Index 0 Param 4=1
    Module.Param0WriteEnabled=true
    Module.Param1WriteEnabled=false
    Module.Param2WriteEnabled=false
    Module.Param3WriteEnabled=true
   )
   &Color(Color=&ColorFromCurve(ColorCurve=DataInterface() {propertyValues {
       AlphaCurve {keys [<at 0 0 RCIM_Cubic>,<at 0.1552 1>,<at 1 0 RCIM_Cubic>]}
       BlueCurve {keys [<at -0.0065 5>,<at 0.0847 30 RCIM_Cubic>,<at 0.4101 5>]}
       GreenCurve {keys [<at -0.0065 0>,<at 0.0847 3.9948 RCIM_Cubic>,<at 0.4101 0>]}
       RedCurve {keys [<at -0.0065 3.0759>,<at 0.0847 18.7895 RCIM_Cubic>,<at 0.4101 3.0759>]}
      }}))
  ]
  rendererClasses [&NiagaraSpriteRendererProperties]
  renderers [{rendererClass=&NiagaraSpriteRendererProperties}]
 }
 FireSmoke_butt {
  emitterUpdateScript [
   &EmitterState(
    Delay First Loop Only=true
    Inactive Response=::NewEnumerator1
    Life Cycle Mode=::NewEnumerator1
    Loop Delay=1
    Loop Duration=1
    UseLoopDelay=true
   )
   &SpawnRate(SpawnRate=6, Use Spawn Probability=false)
  ]
  particleSpawnScript [
   &InitializeParticle(
    Color=<rgba 0 0.1773 3 1>
    Lifetime=&UniformRangedFloat(Minimum=1, Maximum=2)
    Module.Write Color=true
    Module.Write Lifetime=true
    Module.Write Mass=true
    Module.Write Material Random=true
    Module.Write Position=false
    Module.Write Scale=false
    Module.Write SpriteRotation=true
    Module.Write SpriteSize=true
    Position=&SimulationPosition()
    Sprite Rotation=&UniformRangedFloat(Minimum=0, Maximum=300)
    Sprite Size=&Vector2DFromFloat(Value=&UniformRangedFloat(Minimum=60, Maximum=100))
   )
   &BoxLocation(
    Box Offset=&FN_FrontEndVector3Modifier(Gameplay=<0 40 50>, Front End=<0 20 -10>)
    Box Size=<40 10 40>
    MaskSpawnGroup=false
    Module.Override Seed=false
    Rotate Box=false
    Surface Only=false
   )
   &AddVelocity(Velocity=&RandomRangeVector(Minimum=<-5 50 -5>, Maximum=<5 70 5>))
  ]
  particleUpdateScript [
   &ParticleState()
   &SolveForcesAndVelocity(Clamp Velocity=false, Limit Acceleration=false)
   &ScaleSpriteSize(Scale Factor=&Vector2DFromFloat(Value=&FloatFromCurve(FloatCurve=DataInterface() {propertyValues {Curve {keys [
          {arriveTangent=6.3444, interpMode=RCIM_Cubic, time=0, value=0.25}
          {arriveTangent=1.3547, interpMode=RCIM_Cubic, leaveTangent=1.3547, tangentMode=RCTM_User, time=0.2884, value=0.5549}
          {interpMode=RCIM_Cubic, leaveTangent=-1.768, time=0.9996, value=1.2}
         ]}}})))
   &SubUVAnimation(
    Sub UV Play Rate:NiagaraFloat=&RandomRangeFloat(Minimum=0.4, Maximum=0.7)
    SubUV Animation Mode:ENiagara_SubUVLookupMode=::NewEnumerator3
    SubUV Start Time Offset:NiagaraFloat=&RandomRangeFloat(Minimum=0, Maximum=1)
   )
   &DynamicMaterialParameters(
    Index 0 Param 1=&FloatFromCurve(FloatCurve=DataInterface() {propertyValues {Curve {keys [{interpMode=RCIM_Cubic, tangentMode=RCTM_User, time=0.4, value=0}, {interpMode=RCIM_Cubic, tangentMode=RCTM_User, time=1, value=-0.75}]}}})
    Index 0 Param 4=1
    Module.Param0WriteEnabled=true
    Module.Param1WriteEnabled=false
    Module.Param2WriteEnabled=false
    Module.Param3WriteEnabled=true
   )
   &CameraOffset(Camera Offset Amount=&RandomRangeFloat(Minimum=-5, Maximum=20))
   &Color(Color=&ColorFromCurve(ColorCurve=DataInterface() {propertyValues {
       AlphaCurve {keys [<at 0 0 RCIM_Cubic>,<at 0.0972 0.5>,<at 0.7935 0.3>,<at 1 0 RCIM_Cubic>]}
       BlueCurve {keys [<at 0.3023 3 RCIM_Cubic>,<at 0.864 3 RCIM_Cubic>]}
       GreenCurve {keys [<at 0.3023 0.1773 RCIM_Cubic>,<at 0.864 0 RCIM_Cubic>]}
       RedCurve {keys [<at 0.3023 0 RCIM_Cubic>,<at 0.864 1.341 RCIM_Cubic>]}
      }}))
  ]
  rendererClasses [&NiagaraSpriteRendererProperties]
  renderers [{rendererClass=&NiagaraSpriteRendererProperties}]
 }
 Light {
  emitterUpdateScript [
   &EmitterState(
    Delay First Loop Only=true
    Inactive Response=::NewEnumerator1
    Life Cycle Mode=::NewEnumerator1
    Loop Delay=&FN_FrontEndFloatModifier(Gameplay=1.3542, Front End=0.8)
    Loop Duration=1
    UseLoopDelay=true
   )
   &SpawnBurst_Instantaneous(Spawn Count=2, Use Spawn Probability=false)
  ]
  particleSpawnScript [
   &FN_SetSpawnGroup(Number of Spawn Groups=2)
   /Niagara/Modules/Spawn/Initialization/V2/InitializeParticle(Color Mode=::NewEnumerator0, UsePositionOffset=false, Write Material Random=true)
  ]
  particleUpdateScript [
   &SkeletalMeshLocation(
    Apply Sampled Orientation to Mesh Particles=true
    Bone / Socket Index=&SpawnGroup()
    Bone Sampling Mode=::NewEnumerator4
    Inherit Velocity=false
    MaskSpawnGroup=false
    Sampled Position Offset=&MaskVectorBySpawnGroup(Masked Vector=<30 0 0>, Passthrough Vector=<-30 1 1>)
    Sampled Position Offset Coordinate Space:ENiagaraMeshTransforms=::NewEnumerator2
    Skeletal Mesh=$User.SkeletalMeshONE
   )
   &ParticleState()
   &FN_Light_Attributes(
    Color=<rgba 20 0 40 1>
    Create Color Attribute=true
    Create Exponent Attribute=true
    Create Radius Attribute=true
    Enable Particle Light Enable=true
    Exponent=3
    Light Enabled=true
    Radius=70
    Scale Color=&FloatFromCurve(FloatCurve=DataInterface() {propertyValues {Curve {keys [<at -0.0035 -0.0243>,<at 0.3197 1.001>,<at 0.7485 1.0023>,<at 0.9977 0.0032>]}}})
   )
  ]
  rendererClasses [&NiagaraLightRendererProperties]
  renderers [{rendererClass=&NiagaraLightRendererProperties, rendererIndex=0}]
 }
 Paws {
  emitterUpdateScript [
   &EmitterState(
    Delay First Loop Only=true
    Inactive Response=::NewEnumerator1
    Life Cycle Mode=::NewEnumerator1
    Loop Delay=&FN_FrontEndFloatModifier(Gameplay=2.05, Front End=1.5)
    Loop Duration=1
    UseLoopDelay=true
   )
   &SpawnRate(
    Spawn Probability=0.9
    SpawnRate=&MakeCustomFloatFromBool(
     Bool=&SetBoolByFloatComparison(A=$System.FN_Integrator001.Value, B=0.2, Comparison Type=::NewEnumerator2)
     True Float=10
     False Float=6
    )
    Use Spawn Probability=true
   )
  ]
  particleSpawnScript [
   &PartitionParticles()
   &InitializeParticle(
    Lifetime=&Multiply_Float(
     A=&MakeCustomFloatFromBool(
      Bool=&SetBoolByFloatComparison(A=$System.FN_Integrator001.Value, B=0.2, Comparison Type=::NewEnumerator2)
      True Float=1.2
      False Float=1.5
     )
     B=&RandomRangeFloat(Minimum=0.45, Maximum=0.6)
    )
    Sprite Size=&Lerp_Vector2(A=<30 -30>, B=<-30 -30>, Alpha=&MakeFloatFromInt(Int=$Particles.PartitionParticles.Partition))
    Sprite Size Mode=::NewEnumerator4
    UsePositionOffset=false
    Write Material Random=true
   )
   &AddVelocity(Velocity=&Lerp_Vector(
     A=<0 300 0>
     B=<1 460 1>
     Alpha=&FloatFromCurve(FloatCurve=DataInterface() {propertyValues {Curve {keys [<at 0.2 0>,<at 0.8 1>]}}}, CurveIndex=$System.FN_Integrator001.Value)
    ))
   &BoxLocation(
    Box Offset=&Add_Vector(
     A=&Lerp_Vector(A=<15 0 -35>, B=<-15 1 -35>, Alpha=&MakeFloatFromInt(Int=$Particles.PartitionParticles.Partition))
     B=&FN_FrontEndVector3Modifier(Gameplay=<0 0 0>, Front End=<10 -50 -60>)
    )
    Box Size=<5 5 0>
    MaskSpawnGroup=false
    Module.Override Seed=false
    Rotate Box=false
    Surface Only=false
   )
  ]
  particleUpdateScript [
   &ParticleState()
   &SolveForcesAndVelocity(Clamp Velocity=false, Limit Acceleration=false)
   &Color(Color=&MakeLinearColorFromVectorAndFloat(
     "Vector (RGB)"=&MakeVectorFromLinearColorRGB(Color=&RandomRangeLinearColor(Minimum=<rgba 0.1639 0.174 1 1>, Maximum=<rgba 0.5831 0.0069 1 1>))
     "Float (Alpha)"=&FloatFromCurve(FloatCurve=DataInterface() {propertyValues {Curve {keys [<at 0 0>,<at 0.5013 0.9748>,<at 0.7099 0.7732>,<at 1 0>]}}}, Scale Curve=0.8)
    ))
   /NiagaraFN/Sprite/FN_SpriteAlignmentAndFacing(SpriteAlignment:Vector3f=<0 1 0>, SpriteFacing:Vector3f=<0 0 1>)
   &CameraOffset(Camera Offset Amount=&RandomRangeFloat(Minimum=2, Maximum=5)) {enabled=false}
   &ScaleSpriteSize(Scale Factor=&Vector2DFromCurve(
     Vector2Curve=DataInterface() {propertyValues {XCurve {keys [<at 0.5 1>,<at 1 0.5>]}, YCurve {keys [<at 0.5 1>,<at 1 0.8>]}}}
     ScaleCurve=<1 1>
    ))
  ]
  rendererClasses [&NiagaraSpriteRendererProperties]
  renderers [{rendererClass=&NiagaraSpriteRendererProperties}]
 }
 Paws_Deploy {
  emitterUpdateScript [
   &EmitterState(Life Cycle Mode=::NewEnumerator1, Loop Behavior=::NewEnumerator1, Loop Duration=1, UseLoopDelay=true)
   &SpawnRate(
    Spawn Probability=&FN_FrontEndFloatModifier(Gameplay=0.9, Front End=0)
    SpawnRate=&MakeCustomFloatFromBool(
     Bool=&SetBoolByFloatComparison(A=$System.FN_Integrator001.Value, B=0.2, Comparison Type=::NewEnumerator2)
     True Float=10
     False Float=15
    )
    Use Spawn Probability=true
   )
  ]
  particleSpawnScript [
   &FN_SetSpawnGroup(Number of Spawn Groups=2)
   /Niagara/Modules/Spawn/Initialization/InitializeParticle(
    Lifetime=&RandomRangeFloat(Minimum=0.2, Maximum=0.45)
    Mesh Scale=<3 3 3>
    Module.Write Color=true
    Module.Write Lifetime=true
    Module.Write Mass=true
    Module.Write Material Random=true
    Module.Write Position=false
    Module.Write Scale=true
    Module.Write SpriteRotation=false
    Module.Write SpriteSize=true
    Position=&SimulationPosition()
    Sprite Size=&MaskVector2dBySpawnGroup(Masked Vector2D=<30 -30>, Passthrough Vector2D=<-30 -30>, Spawn Group Mask=0)
   )
   &AddVelocity(Velocity=<0 450 0>)
   &BoxLocation(
    Box Offset=&MaskVectorBySpawnGroup(Masked Vector=<20 0 -35>, Passthrough Vector=<-20 1 -35>)
    Box Size=<5 5 5>
    MaskSpawnGroup=false
    Module.Override Seed=false
    Rotate Box=false
    Surface Only=false
   )
   &SkeletalMeshLocation(
    Bone / Socket Index=1
    Inherit Velocity=false
    MaskSpawnGroup=false
    Mesh Sampling Type=::NewEnumerator5
    Sampled Position Offset:NiagaraPosition=$Particles.Position
    Skeletal Mesh=$User.SkeletalMeshONE
    Socket Sampling Mode=::NewEnumerator5
   )
  ]
  particleUpdateScript [
   &ParticleState()
   &SolveForcesAndVelocity(Clamp Velocity=false, Limit Acceleration=false)
   &Color(Color=&MakeLinearColorFromVectorAndFloat(
     "Vector (RGB)"=&MakeVectorFromLinearColorRGB(Color=&RandomRangeLinearColor(Minimum=<rgba 0.1639 0.174 1 1>, Maximum=<rgba 0.5831 0.0069 1 1>))
     "Float (Alpha)"=&FloatFromCurve(FloatCurve=DataInterface() {propertyValues {Curve {keys [<at 0 0>,<at 0.5013 0.9748>,<at 0.7099 0.7732>,<at 1 0>]}}}, Scale Curve=0.8)
    ))
   &FN_SpriteAlignmentAndFacing(SpriteAlignment:Vector3f=<0 1 0>, SpriteFacing:Vector3f=<0 0 1>)
   &CameraOffset(Camera Offset Amount=&RandomRangeFloat(Minimum=2, Maximum=5))
   &ScaleSpriteSize(Scale Factor=&Vector2DFromCurve(
     Vector2Curve=DataInterface() {propertyValues {XCurve {keys [<at 0.5 1>,<at 1 0.5>]}, YCurve {keys [<at 0.5 1>,<at 1 0.8>]}}}
     ScaleCurve=<1 1>
    ))
  ]
  rendererClasses [&NiagaraSpriteRendererProperties]
  renderers [{rendererClass=&NiagaraSpriteRendererProperties}]
 }
 SmokeUnder {
  emitterUpdateScript [
   &EmitterState(
    Delay First Loop Only=true
    Life Cycle Mode=::NewEnumerator1
    Loop Delay=&FN_FrontEndFloatModifier(Gameplay=2.05, Front End=1.5)
    Loop Duration=1
    UseLoopDelay=true
   )
   &SpawnRate(
    Spawn Probability=0.9
    SpawnRate=&MakeCustomFloatFromBool(
     Bool=&SetBoolByFloatComparison(A=$System.FN_Integrator001.Value, B=0.2, Comparison Type=::NewEnumerator2)
     True Float=25
     False Float=15
    )
    Use Spawn Probability=true
   )
  ]
  particleSpawnScript [
   /Niagara/Modules/Spawn/Initialization/V2/InitializeParticle(
    Lifetime=&Multiply_Float(
     A=&MakeCustomFloatFromBool(
      Bool=&SetBoolByFloatComparison(A=$System.FN_Integrator001.Value, B=0.2, Comparison Type=::NewEnumerator2)
      True Float=1
      False Float=1.2
     )
     B=&RandomRangeFloat(Minimum=0.35, Maximum=0.55)
    )
    Sprite Rotation Mode=::NewEnumerator2
    Sprite Size Mode=::NewEnumerator1
    Uniform Sprite Size Max=130
    Uniform Sprite Size Min=80
    UsePositionOffset=false
    Write Material Random=true
   )
   &AddVelocity(Velocity=&Lerp_Vector(
     A=<0 300 0>
     B=<1 460 1>
     Alpha=&FloatFromCurve(FloatCurve=DataInterface() {propertyValues {Curve {keys [<at 0.2 0>,<at 0.8 1>]}}}, CurveIndex=$System.FN_Integrator001.Value)
    ))
   &BoxLocation(
    Box Offset=&Add_Vector(A=<0 -20 -35>, B=&FN_FrontEndVector3Modifier(Gameplay=<0 0 0>, Front End=<10 -50 -50>))
    Box Size=<20 10 10>
    MaskSpawnGroup=false
    Module.Override Seed=false
    Rotate Box=false
    Surface Only=false
   )
   &DynamicMaterialParameters(
    Index 0 Param 1=0.6
    Index 0 Param 2=6
    Module.Param0WriteEnabled=true
    Module.Param1WriteEnabled=false
    Module.Param2WriteEnabled=true
    Module.Param3WriteEnabled=true
   )
   &SubUVAnimation(End Frame=3, SubUV Animation Mode:ENiagara_SubUVLookupMode=::NewEnumerator2)
  ]
  particleUpdateScript [
   &ParticleState()
   &SolveForcesAndVelocity(Clamp Velocity=false, Limit Acceleration=false)
   &Color(Color=&MakeLinearColorFromVectorAndFloat(
     "Vector (RGB)"=&MakeVectorFromLinearColorRGB(Color=&RandomRangeLinearColor(Minimum=<rgba 0.1639 0.174 1 1>, Maximum=<rgba 0.5831 0.0069 1 1>))
     "Float (Alpha)"=&FloatFromCurve(FloatCurve=DataInterface() {propertyValues {Curve {keys [<at 0 0>,<at 0.5013 0.9748>,<at 0.7099 0.7732>,<at 1 0>]}}}, Scale Curve=0.5)
    ))
   &FN_SpriteAlignmentAndFacing(SpriteAlignment:Vector3f=<0 1 0>, SpriteFacing:Vector3f=<0 0 1>)
   &CameraOffset(Camera Offset Amount=&RandomRangeFloat(Minimum=-5, Maximum=5))
   &ScaleSpriteSize(Scale Factor=&Vector2DFromFloat(Value=&FloatFromCurve(FloatCurve=DataInterface() {propertyValues {Curve {keys [<at 0 0>,<at 0.3268 0.6265>,<at 0.8076 0.937>,<at 1 0>]}}})))
  ]
  rendererClasses [&NiagaraSpriteRendererProperties]
  renderers [{rendererClass=&NiagaraSpriteRendererProperties}]
 }
 Trails_Purple {
  emitterUpdateScript [
   &EmitterState(
    Delay First Loop Only=true
    Life Cycle Mode=::NewEnumerator1
    Loop Delay=&FN_FrontEndFloatModifier(Gameplay=2.05, Front End=1.5)
    Loop Duration=1
    UseLoopDelay=true
   )
   &FN_SpawnRibbon_v2(Number of Ribbons=13, Spawn Mode=::NewEnumerator0, SpawnRate=60)
  ]
  particleSpawnScript [
   &FN_SetSpawnGroup(Number of Spawn Groups=13)
   /Game/Effects/Niagara/Modules/Ribbons/FN_InitializeRibbon_v4(Enable Skeletal Position=true, Lifetime=0.08, Ribbon Width=25, SkeletalMesh:NiagaraDataInterfaceSkeletalMesh=$User.SkeletalMesh)
   &FN_FlipRibbonV()
   &AddVelocity(Velocity=&RotateVectorByQuaternion(
     Vector To Rotate=&Lerp_Vector(
      A=<0 2500 0>
      B=<0 3000 0>
      Alpha=&FloatFromCurve(
       FloatCurve=DataInterface() {propertyValues {Curve {keys [<at 0.2 0>,<at 0.8 1>]}}}
       CurveIndex=&FN_FrontEndFloatModifier(Gameplay=$System.FN_Integrator001.Value, Front End=1)
      )
     )
     Quaternion=$Particles.FN_UpdateRibbon_v5.SampledRotation
    ))
  ]
  particleUpdateScript [
   &SolveForcesAndVelocity(Clamp Velocity=false, Limit Acceleration=false)
   /Game/Effects/Niagara/Modules/Ribbons/FN_UpdateRibbon_v5(Enable Skeletal Position=true, SkeletalMesh:NiagaraDataInterfaceSkeletalMesh=$User.SkeletalMesh)
   &ScaleColor(
    Scale Alpha=&FN_FrontEndFloatModifier(Gameplay=&Lerp_Float(A=0.85, B=1, Alpha=$System.FN_Integrator001.Value), Front End=1)
    Scale RGB=&MakeVectorFromLinearColorRGB(Color=&ColorFromCurve(ColorCurve=DataInterface() {propertyValues {
        AlphaCurve {keys [<at 0 1 RCIM_Cubic>,<at 0.6445 1>,<at 1 0 RCIM_Cubic>]}
        BlueCurve {keys [<at 0 3 RCIM_Cubic>,<at 0.8753 3 RCIM_Cubic>]}
        GreenCurve {keys [<at 0 0.1773 RCIM_Cubic>,<at 0.8753 0 RCIM_Cubic>]}
        RedCurve {keys [<at 0 0 RCIM_Cubic>,<at 0.8753 2.1032 RCIM_Cubic>]}
       }}))
    ScaleA=true
    ScaleRGB=true
   )
  ]
  rendererClasses [&NiagaraRibbonRendererProperties]
  renderers [{rendererClass=&NiagaraRibbonRendererProperties}]
 }
 Trails_PurpleDeploy {
  emitterUpdateScript [
   &EmitterState(
    Inactive Response=::NewEnumerator1
    Life Cycle Mode=::NewEnumerator1
    Loop Behavior=::NewEnumerator1
    Loop Duration=1.9083
    UseLoopDelay=false
   )
   &FN_SpawnRibbon_v2(Number of Ribbons=13, Spawn Mode=::NewEnumerator0, SpawnRate=60)
  ]
  particleSpawnScript [
   &FN_SetSpawnGroup(Number of Spawn Groups=13)
   &FN_InitializeRibbon_v4(Enable Skeletal Position=true, Lifetime=0.08, Ribbon Width=25, SkeletalMesh:NiagaraDataInterfaceSkeletalMesh=$User.SkeletalMesh)
   &FN_FlipRibbonV()
   &SkeletalMeshLocation(
    Apply To Particle Position=false
    Inherit Velocity=false
    MaskSpawnGroup=false
    Mesh Sampling Type=::NewEnumerator5
    Skeletal Mesh=$User.SkeletalMeshONE
    Socket Sampling Mode=::NewEnumerator5
    Use Sampled Mesh Attribute as Ribbon ID=false
   ) {module=SkeletalMeshLocation001}
   &AddVelocity(
    Coordinate Space=::World
    Velocity=&RotateVectorByQuaternion(
     Vector To Rotate=&Lerp_Vector(
      A=<5000 -1000 0>
      B=<5000 0 0>
      Alpha=&FloatFromCurve(FloatCurve=DataInterface() {propertyValues {Curve {keys [<at 0.2 0>,<at 0.8 1>]}}}, CurveIndex=$System.FN_Integrator001.Value)
     )
     Quaternion=$Output.SkeletalMeshLocation001.SampledRotation
    )
   ) {module=AddVelocity001}
  ]
  particleUpdateScript [
   &ParticleState()
   &SolveForcesAndVelocity(Clamp Velocity=false, Limit Acceleration=false)
   &FN_UpdateRibbon_v5(Enable Skeletal Position=true, SkeletalMesh:NiagaraDataInterfaceSkeletalMesh=$User.SkeletalMesh)
   &ScaleColor(
    Scale Alpha=&FloatFromCurve(FloatCurve=DataInterface() {propertyValues {Curve {keys [<at 0 1>,<at 0.8 1>,<at 1 0>]}}}, CurveIndex=$Emitter.NormalizedLoopAge)
    Scale RGB=&MakeVectorFromLinearColorRGB(Color=&ColorFromCurve(ColorCurve=DataInterface() {propertyValues {
        AlphaCurve {keys [<at 0 1 RCIM_Cubic>,<at 0.6445 1>,<at 1 0 RCIM_Cubic>]}
        BlueCurve {keys [<at 0 3 RCIM_Cubic>,<at 0.8753 3 RCIM_Cubic>]}
        GreenCurve {keys [<at 0 0.1773 RCIM_Cubic>,<at 0.8753 0 RCIM_Cubic>]}
        RedCurve {keys [<at 0 0 RCIM_Cubic>,<at 0.8753 2.1032 RCIM_Cubic>]}
       }}))
    ScaleA=true
    ScaleRGB=true
   )
  ]
  rendererClasses [&NiagaraRibbonRendererProperties]
  renderers [{rendererClass=&NiagaraRibbonRendererProperties}]
 }
}
summary {
 systemName=NS_Glider_AncientGliadiator_M_Idle
 emitters [
  {emitterName=Paws, rendererClasses [&NiagaraSpriteRendererProperties]}
  {emitterName=Trails_Purple, rendererClasses [&NiagaraRibbonRendererProperties]}
  {emitterName=SmokeUnder, rendererClasses [&NiagaraSpriteRendererProperties]}
  {emitterName=Trails_PurpleDeploy, rendererClasses [&NiagaraRibbonRendererProperties]}
  {emitterName=Paws_Deploy, rendererClasses [&NiagaraSpriteRendererProperties]}
  {emitterName=FireRing, rendererClasses [&NiagaraMeshRendererProperties]}
  {emitterName=Light, rendererClasses [&NiagaraLightRendererProperties]}
  {emitterName=Decal001, rendererClasses [&NiagaraSpriteRendererProperties]}
  {emitterName=FireRing_Deploy, rendererClasses [&NiagaraMeshRendererProperties]}
  {emitterName=FireSmoke, rendererClasses [&NiagaraSpriteRendererProperties]}
  {emitterName=FireSmoke_butt, rendererClasses [&NiagaraSpriteRendererProperties]}
  {emitterName=FireSmoke_Forward_Deploy, rendererClasses [&NiagaraSpriteRendererProperties]}
 ]
 userVariables [
  User.bIsFrontEnd {type=NiagaraBool}
  User.bIsFrontEndPreview {type=NiagaraBool}
  User.ColorChange {type=LinearColor, description="", defaultValue {struct=LinearColor, value=<rgba 1 1 1 1>}}
  User.ForwardDot {type=NiagaraFloat, defaultValue=0, description=""}
  User.IsFullyDeployed {type=NiagaraFloat, defaultValue=0, description=""}
  User.IsSkydiving {type=NiagaraBool}
  User.RibbonIDOffset {type=NiagaraInt32}
  User.RightDot {type=NiagaraFloat, defaultValue=0, description=""}
  User.SkeletalMesh {
   type=NiagaraDataInterfaceSkeletalMesh
   defaultValue {
    struct=NiagaraExt_VariableValue_DataInterface
    value {dataInterface=&self:NiagaraDataInterfaceSkeletalMesh_0}
   }
  }
  User.SkeletalMeshONE {
   type=NiagaraDataInterfaceSkeletalMesh
   defaultValue {
    struct=NiagaraExt_VariableValue_DataInterface
    value {dataInterface=&self:NiagaraDataInterfaceSkeletalMesh_1}
   }
  }
 ]
}
